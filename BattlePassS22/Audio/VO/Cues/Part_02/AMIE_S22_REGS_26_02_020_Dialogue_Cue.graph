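# Sound cue for one Amie voice-over line from Season 22 Battle Pass content, dialogue AMIE_S22_REGS_26_02_020.
# A single dialogue player node plays the dialogue wave with the Amie voice as speaker and no targets.
# The cue uses the S22 quest attenuation, the narrator VO sound class, boosted volume and a very large max distance.
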SoundCue /BattlePassS22/Audio/VO/Cues/Part_02/AMIE_S22_REGS_26_02_020_Dialogue_Cue
root=DialoguePlayer_0
nodes [{
  id=DialoguePlayer_0
  properties {dialogueWaveParameter {
    dialogueWave=/BattlePassS22/Audio/VO/Waves/Amie/AMIE_S22_REGS_26_02_020_Dialogue
    context {speaker=/BattlePassS22/Audio/VO/Amie_DV, targets []}
   }}
 }]
properties {
 attenuationSettings=/BattlePassS22/Audio/VO/S22_Quest_Att
 maxDistance=2097152
 soundClassObject=/Game/Sounds/FortSoundClassesAndModes/Fort_VO_Narrator
 volumeMultiplier=2
}
